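# Material function that builds a three-axis grayscale gradient across a mesh's pre-skinned local bounds, so the gradient does not shift with skinned deformation.
# Vertex-interpolated pre-skinned local position is offset by the bounds minimum, divided by the extents and saturated to 0–1 per axis.
# Min and Max inputs remap that range via inverse lerp, then the result is saturated and shaped by a Pow exponent.
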MaterialFunction /Game/Global/Materials/MaterialGym/Materials/MaterialElements/GrayscaleTransforms/BoundsGradient/MF_PreskinnedGradient
graph {nodes [
  n1 = FunctionInput() {bUsePreviewValueAsDefault=true, inputName=Max, inputType=FunctionInput_Scalar, sortPriority=1, previewValue=<1 0 0 1>}
  n2 = PreSkinnedLocalBounds()
  n3 = Subtract(A=VertexInterpolator(VS=LocalPosition() {localOrigin=InstancePreSkinning}.XYZ).PS, B=@n2.Min)
  n4 = InverseLinearInterpolate(
   A=FunctionInput() {bUsePreviewValueAsDefault=true, inputName=Min, inputType=FunctionInput_Scalar}
   B=@n1
   Value=Saturate(None=Divide(A=@n3, B=@n2.Extents))
  )
  n5 = FunctionInput() {bUsePreviewValueAsDefault=true, inputName=Pow, inputType=FunctionInput_Scalar, sortPriority=3, previewValue=<1 0 0 1>}
  FunctionOutput(None=Power(Base=Saturate(None=@n4), Exp=@n5)) {outputName="Gradient (V3)"}
 ]}
