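# A Material Gym material for skeletal meshes and mesh deformers that demonstrates a two-color gradient, named for pre-skinned bounds.
# The MF_PreskinnedGradient function supplies a three-component gradient, from which a mask keeps only the blue (Z) channel.
# That channel drives a lerp between the Color01 and Color02 parameters, which default to black and white.
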Material /Game/Global/Materials/MaterialGym/Materials/MaterialElements/GrayscaleTransforms/BoundsGradient/M_PreskinnedGradient
graph {
 properties {bAutomaticallySetUsageInEditor=false, bUsedWithMeshDeformer=true, bUsedWithSkeletalMesh=true, usedShadingModels=Default Lit}
 nodes [LinearInterpolate(
   A=VectorParameter() {parameterName=Color01, defaultValue=<rgba 0 0 0 1>}.RGB
   B=VectorParameter() {parameterName=Color02, defaultValue=<rgba 1 1 1 1>}.RGB
   Alpha=ComponentMask(None=MaterialFunctionCall!(Min, Max, Pow) {materialFunction=&/MF_PreskinnedGradient}."Gradient (V3)") {b=true}
  )]
}
